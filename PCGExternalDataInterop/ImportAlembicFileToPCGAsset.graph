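# Editor Utility Widget that exports Alembic files to PCG assets.
# Its details view shows the widget's own PCG-category properties, including import settings.
# One button applies the standard CitySample setup to those settings; the other calls ExportAlembicFileToPCG with a save dialog.
# An unconnected manual chain, triggered by no event, loads or creates PCGBlueprintSettings, fills its PCGAsset element with Alembic data and a Points pin, then saves.
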EditorUtilityWidgetBlueprint /PCGExternalDataInterop/ImportAlembicFileToPCGAsset
graphs [EventGraph {nodes [
   n1 = Get OutputPath() -> (OutputPath:Directory Path Structure)
   n2 = FilenameToPackageName!(
    execute
    Filename:String=Append!(
     A=Break Directory Path!(DirectoryPath:"Directory Path Structure (by ref)"=@n1) -> (Path:String) {is=title K2Node_BreakStruct}
     B=/
     C:String=Get OutputFileName() -> (OutputFileName:String)
    ) -> (ReturnValue)
   ) -> (then, ReturnValue:String)
   > Set PCGAssetName(PCGAssetName:String=GetBaseFilename(InPath=@n2, bRemovePath=true)) -> (Output_Get:String)
   > n3 = DoesAssetExist(AssetPath=@n2)
   > n4 = Branch(Condition=@n3)
   > n5 = LoadAsset(AssetPath=@n2)
   > Set Asset(Asset:Object Reference=@n5) -> (Output_Get:Object Reference)
   > n6 = Branch(Condition=IsValid(Object=Get Asset() -> (Asset:Object Reference)))
   > n7 = Cast To PCGBlueprintSettings!(execute, Object=Get Asset()) -> (then, CastFailed, AsPCGBlueprint Settings:PCGBlueprint Settings Object Reference) {is=title K2Node_DynamicCast}
   > n8 = Cast To PCGAsset!(
    execute
    Object=Get BlueprintElementInstance!(self:PCGBlueprint Settings Object Reference=@n7) -> (BlueprintElementInstance:PCGBlueprint Base Element Object Reference)
   ) -> (then, CastFailed, AsPCGAsset:PCGAsset Object Reference) {is=title K2Node_DynamicCast}
   > n9 = Set As PCGAsset(As PCGAsset:PCGAsset Object Reference=@n8) -> (Output_Get:PCGAsset Object Reference)
   n10 = LoadAlembicFileToPCG!(
    execute=[@n17.then, @n9.then]
    Settings:"PCGLoad Alembic BPData Structure (by ref)"=Get ImportSettings() -> (ImportSettings:PCGLoad Alembic BPData Structure)
    TargetOuter:Object Reference=Get Asset()
   ) -> (then, Data:PCGData Collection Structure)
   n11 = Get As PCGAsset() -> (As PCGAsset:PCGAsset Object Reference)
   > Set PCGDataCollection!(execute, PCGDataCollection:PCGData Collection Structure=@n10, self:PCGAsset Object Reference=@n11) -> (then, Output_Get:PCGData Collection Structure)
   > Set NodeTitle!(execute, NodeTitle:Name=String To Name(InString=Get PCGAssetName() -> (PCGAssetName:String)), self:PCGAsset Object Reference=@n11) -> (then, Output_Get:Name=None)
   > Set CustomOutputPins!(
    execute
    CustomOutputPins:Array of PCGPin Properties Structures=Make Array("[0]":PCGPin Properties Structure=MakePinProperty!(
      Label:Name=Points
      bAllowMultipleData:bool=true
      bAllowMultipleConnections:bool=false
      bIsAdvancedPin:bool=false
      AllowedType:EPCGExclusiveDataType Enum=Point
     ) -> (ReturnValue:PCGPin Properties Structure)) -> (Array:Array of PCGPin Properties Structures)
    self:PCGBlueprint Base Element Object Reference=Get As PCGAsset()
   ) -> (then, Output_Get:Array of PCGPin Properties Structures)
   > Set bHasDefaultInPin!(execute, bHasDefaultInPin:bool=false, self:PCGBlueprint Base Element Object Reference=Get As PCGAsset()) -> (then, Output_Get:bool)
   > Set bHasDefaultOutPin!(execute, bHasDefaultOutPin:bool=false, self:PCGBlueprint Base Element Object Reference=Get As PCGAsset()) -> (then, Output_Get:bool)
   > SaveLoadedAsset(AssetToSave=Get As PCGAsset(), bOnlyIfIsDirty=false)
   PrintText(execute=@n8.CastFailed, InText="INVTEXT(\"Asset is not a PCG asset\")")
   PrintText(execute=@n7.CastFailed, InText="INVTEXT(\"Asset is not a valid PCG blueprint settings\")")
   PrintText(execute=@n6.else, InText="INVTEXT(\"Unable to load asset\")")
   n12 = GetAssetTools(execute=@n4.else)
   > n13 = CreateAsset(
    self=@n12
    AssetName=GetCleanFilename!(InPath:String=@n2) -> (ReturnValue:String)
    PackagePath=GetPath!(InPath:String=@n2) -> (ReturnValue:String)
    AssetClass=/Script/PCG.PCGBlueprintSettings
   )
   > Set Asset(Asset=@n13)
   > n14 = Branch(Condition=IsValid(Object=Get Asset()))
   > n15 = Cast To PCGBlueprintSettings(Object=Get Asset())
   > n16 = SetElementType!(
    execute
    self:PCGBlueprint Settings Object Reference=@n15
    InElementType:Deprecated Blueprint Element Class Reference=/PCG/BP_Elements/PCGAsset.PCGAsset_C
   ) -> (then, ElementInstance:PCGAsset Object Reference)
   > n17 = Set As PCGAsset(As PCGAsset=@n16)
   PrintText(execute=@n15.CastFailed, InText="INVTEXT(\"Asset type is not PCG Asset\")")
   PrintText(execute=@n14.else, InText="INVTEXT(\"Unable to create asset\")")
   "On Clicked (EditorUtilityButton_1)"() {is=title K2Node_ComponentBoundEvent}
   > ExportAlembicFileToPCG!(
    execute
    Settings:"PCGLoad Alembic BPData Structure (by ref)"=Get ImportSettings()
    Parameters:PCGAsset Exporter Parameters Structure=(bOpenSaveDialog=True,AssetName="",AssetPath="",bSaveOnExportEnded=True)
   ) -> (then)
   "On Clicked (EditorUtilityButton_515)"() {is=title K2Node_ComponentBoundEvent}
   > SetupFromStandard!(execute, Data:"PCGLoad Alembic BPData Structure (by ref)"=Get ImportSettings(), InSetup:EPCGLoadAlembicStandardSetup Enum=CitySample) -> (then)
   Event Construct()
   > SetObject(self=Get DetailsView_1() -> (DetailsView_1:Details View Object Reference), NewObject=Self-Reference())
   Event PreConstruct()
   Event Tick()
  ]}]
variables [
 As PCGAsset {type=PCGAsset_C, default=null}
 Asset {type=Object, default=null}
 ImportSettings {
  type=PCGLoadAlembicBPData
  default {
   bConversionFlipHandedness=false
   alembicFilePath {filePath=""}
   attributeMapping {}
   conversionSettings {bFlipU=false, bFlipV=true, preset=Maya, rotation=<0 0 0>, scale=<1 -1 1>}
  }
 }
 OutputFileName {type=string, default=""}
 OutputPath {type=DirectoryPath, default {path=""}}
 PCGAssetName {type=string, default=""}
 PCGData Collection {type=PCGDataCollection, default {bCancelExecutionOnEmpty=false, taggedData []}}
]
widgets {
 info {rootWidgetClass=&VerticalBox, widgetCount=6}
 widgets [
  VerticalBox_80 {widgetClassPath=&VerticalBox, widget=VerticalBox_80}
  DetailsView_1 {
   widgetClassPath=&DetailsView
   parent=VerticalBox_80
   widget=DetailsView_1
   properties {categoriesToShow [PCG]}
   slot {class=VerticalBoxSlot}
  }
  EditorUtilityButton_515 {widgetClassPath=&EditorUtilityButton, parent=VerticalBox_80, widget=EditorUtilityButton_515, slot {class=VerticalBoxSlot}}
  TextBlock_145 {
   widgetClassPath=&TextBlock
   parent=EditorUtilityButton_515
   widget=TextBlock_145
   properties {text=Use standard CitySample setup}
   slot {class=ButtonSlot, properties {horizontalAlignment=HAlign_Fill, padding=<ltrb 0 0 0 0>}}
  }
  EditorUtilityButton_1 {widgetClassPath=&EditorUtilityButton, parent=VerticalBox_80, widget=EditorUtilityButton_1, slot {class=VerticalBoxSlot}}
  TextBlock_41 {
   widgetClassPath=&TextBlock
   parent=EditorUtilityButton_1
   widget=TextBlock_41
   properties {text=Export to PCG Asset}
   slot {class=ButtonSlot, properties {horizontalAlignment=HAlign_Fill, verticalAlignment=VAlign_Fill, padding=<ltrb 0 0 0 0>}}
  }
 ]
}
